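# EQS query for the Demon Arena that ranks candidate level injection locations by closeness to the safe zone center.
# The generator gathers every BP_DemonArena_LevelInjectionLocation actor without a radius limit, with result randomization enabled.
# A single score-only 3D distance test measures against the Athena safe zone center index context, using inverse-linear scoring so nearer locations rank higher.
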EnvQuery /DemonArenaGameplay/Gameplay/EQS/EQS_DemonArenaInjection
options [{
  generator {
   name=FortEnvQueryGenerator_ActorsOfClass_0
   class=FortEnvQueryGenerator_ActorsOfClass
   properties {
    bAutoSortTests=true
    bCanRunAsync=false
    bRandomizeResults=true
    envQueryResultNormalizationOption=Default
    generateOnlyActorsInRadius=false
    optionName=""
    searchCenter=&EnvQueryContext_Querier
    searchRadius=500
    searchedActorClass=/DemonArenaGameplay/Gameplay/Actors/BP_DemonArena_LevelInjectionLocation.BP_DemonArena_LevelInjectionLocation_C
   }
  }
  tests [EnvQueryTest_Distance_0 {
    class=EnvQueryTest_Distance
    properties {
     distanceTo=/Script/FortniteAIServer.FortQueryContext_AthenaSafeZoneCenterIndex
     filterType=Range
     floatValueMin=0
     scoringEquation=InverseLinear
     scoringFactor=1
     testComment=""
     testMode=Distance3D
     testPurpose=Score
    }
   }]
 }]
